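# Translucent UI material for a "Last Dash" indicator, combining a pip shape with a lightning glyph and a fill color.
# The pip is an SDF box built from offset, scaled, skewed and rotated UVs, and the lightning is a smoothstepped SDF texture.
# A vertical progress edge animates between start and end parameters via AnimateProgress.
# Named reroutes feed lerps that blend black, white and FillColor, and weight shape opacities.
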Material /ScubaBondOakGameplay/UI/Materials/M_UI_LastDash
graph {
 properties {
  bAutomaticallySetUsageInEditor=false
  blendMode=BLEND_Translucent
  materialDomain=MD_UI
  pixelDepthOffsetMode=PDOM_Legacy
  usedShadingModels=Default Lit
 }
 nodes [
  n1 = NamedRerouteUsage()
  LinearInterpolate(
   A=Multiply(A=NamedRerouteUsage(), B=ScalarParameter() {defaultValue=0.3, parameterName=PipOpacity})
   B=Multiply(A=@n1, B=ScalarParameter() {parameterName=LightningOpacity})
   Alpha=@n1
  )
  LinearInterpolate(
   A=Constant3Vector()
   B=LinearInterpolate(
    A=Constant3Vector() {constant=<rgba 1 1 1 0>}
    B=VectorParameter() {parameterName=FillColor, defaultValue=<rgba 0 1 1 0>}.RGB
    Alpha=NamedRerouteUsage()
   )
   Alpha=NamedRerouteUsage()
  )
  n2 = TextureCoordinate()
  n3 = ComponentMask(None=@n2) {r=true}
  n4 = ScalarParameter() {parameterName=TopPipX}
  n5 = ComponentMask(None=@n2) {g=true}
  n6 = ScalarParameter() {defaultValue=-0.064, parameterName=TopPipY}
  n7 = MaterialFunctionCall(UVs=AppendVector(A=Add(A=@n3, B=@n4), B=Add(A=@n5, B=@n6)), Texture Scale=ScalarParameter() {defaultValue=0.9, parameterName=PipScale}) {materialFunction=&ScaleUVsByCenter}
  n8 = Constant2Vector() {r=10}
  n9 = MaterialFunctionCall(
   UVs=MaterialFunctionCall(
    UVs=MaterialFunctionCall!(UV=@n7.UVs, Skew Angle=@n8) {materialFunction=/FortUILibrary/MaterialFunctions/Basics/MF_UI_Skew}.Skewed UVs
    "Rotation Angle (0-1)"=Constant() {r=0.125}
   ) {materialFunction=&CustomRotator}.Rotated Values
   Size=ScalarParameter() {defaultValue=0.6, parameterName=PipSize}
   OutlineThickness=Constant() {r=0.028}
  ) {materialFunction=&MF_UI_Box}
  NamedRerouteDeclaration(None=OneMinus(None=SmoothStep(Value=@n9.SDF) {constMax=0.001})) {name=Pip, nodeColor=<rgba 0 0 0 1>}
  n10 = TextureCoordinate()
  n11 = ComponentMask(None=@n10) {r=true}
  n12 = ScalarParameter() {defaultValue=-0.1, parameterName=LightningX}
  n13 = ComponentMask(None=@n10) {g=true}
  n14 = ScalarParameter() {defaultValue=0.7, parameterName=LightningY}
  n15 = MaterialFunctionCall(UVs=AppendVector(A=Add(A=@n11, B=@n12), B=Add(A=@n13, B=@n14)), Texture Scale=ScalarParameter() {parameterName=ScaleTexture}) {materialFunction=&ScaleUVsByCenter}
  NamedRerouteDeclaration(None=SmoothStep(Value=TextureSample(UVs=@n15.UVs) {samplerSource=SSM_Clamp_WorldGroupSettings, texture=/ScubaBondOakGameplay/UI/Textures/T_UI_Lightning_SDF}.R) {constMax=0.656, constMin=0.464}) {name=LightningTexture, nodeColor=<rgba 0 0 0 1>}
  n16 = ScalarParameter() {group=Progress_Layout, parameterName=ProgressEdgeSharpness, sliderMax=1, sortPriority=2}
  n17 = MaterialFunctionCall(UVs=TextureCoordinate(), "Rotation Angle (0-1)"=Constant() {r=0.5}) {materialFunction=&CustomRotator}
  n18 = ScalarParameter() {
   defaultValue=0.15
   desc="When progress value changes, it defines where change starts.\r\n0 - 1"
   group=Progress
   parameterName=ProgressStart
   sliderMax=1
   sortPriority=1
  }
  n19 = ScalarParameter() {
   defaultValue=0.96
   desc="When progress value changes, it defines where change ends.\r\n0 - 1"
   group=Progress
   parameterName=ProgressEnd
   sliderMax=1
   sortPriority=2
  }
  n20 = ScalarParameter() {
   desc="Property for animating progress value change (from ProgressStart to ProgressEnd)\r\n0 - 1"
   group=Progress
   parameterName=AnimateProgress
   sliderMax=1
   sortPriority=3
  }
  n21 = LinearInterpolate(A=@n18, B=@n19, Alpha=@n20)
  n22 = ComponentMask(None=Add(A=Multiply(A=@n17.Rotated Values) {constB=-1}, B=NamedRerouteDeclaration(None=@n21) {name=Progress, nodeColor=<rgba 0 0 0 1>}) {constB=0.5}) {g=true}
  NamedRerouteDeclaration(None=Saturate(None=SmoothStep(Max=@n16, Value=@n22) {constMax=0.01})) {name=ProgressGraident, nodeColor=<rgba 0 0 0 1>}
  NamedRerouteUsage()
 ]
}
