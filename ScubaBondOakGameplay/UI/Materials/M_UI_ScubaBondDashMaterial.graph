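# Translucent UI material for the ScubaBond dash indicator, drawing skewed diamond segments tiled three times vertically, plus a separate pip and horizontal line streaks.
# SDF box functions with skew and rotation produce the fill, outline and pip masks, which composite with adjustable colors and opacities.
# Animatable start/end lerps define vertical and horizontal progress gradients.
# Panning Perlin noise distorts a gradient-faded progress line.
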Material /ScubaBondOakGameplay/UI/Materials/M_UI_ScubaBondDashMaterial
graph {
 properties {
  bAutomaticallySetUsageInEditor=false
  blendMode=BLEND_Translucent
  materialDomain=MD_UI
  pixelDepthOffsetMode=PDOM_Legacy
  usedShadingModels=Default Lit
 }
 nodes [
  Constant2Vector() {g=0.3, r=0.3}
  Constant2Vector()
  Constant() {r=0.1}
  Constant() {r=1.2298}
  Constant() {r=0.04}
  Constant() {r=1.2298}
  n1 = NamedRerouteUsage()
  n2 = ScalarParameter() {parameterName=PipOpacity}
  n3 = Subtract(A=Add(A=Multiply(A=@n1, B=@n2), B=NamedRerouteUsage()), B=NamedRerouteUsage())
  LinearInterpolate(
   A=LinearInterpolate(
    A=VectorParameter() {parameterName=MainColor, defaultValue=<rgba 0 1 1 0>}.RGB
    B=Constant3Vector() {constant=<rgba 1 1 1 1>}
    Alpha=Saturate(None=@n3)
   )
   B=VectorParameter() {parameterName=SquresOutlineColor, defaultValue=<rgba 0 0 0 1>}.RGB
   Alpha=NamedRerouteUsage()
  )
  n4 = Add(
   A=Multiply(A=NamedRerouteUsage(), B=NamedRerouteUsage())
   B=Multiply(A=NamedRerouteUsage(), B=ScalarParameter() {defaultValue=0.3, parameterName=OutlineOpacity})
  )
  n5 = Multiply(A=NamedRerouteUsage(), B=Multiply(A=NamedRerouteUsage()) {constB=0.5})
  n6 = NamedRerouteUsage()
  n7 = NamedRerouteUsage()
  n8 = NamedRerouteUsage()
  LinearInterpolate(
   A=LinearInterpolate(A=Add(A=@n4, B=@n5), B=Multiply(A=@n6, B=@n7), Alpha=@n6)
   B=@n8
   Alpha=Multiply(A=@n8, B=ScalarParameter() {parameterName=PipOpacity})
  )
  n9 = MaterialFunctionCall!(UV=NamedRerouteUsage(), Skew Angle=Constant2Vector() {r=10}) {materialFunction=/FortUILibrary/MaterialFunctions/Basics/MF_UI_Skew}
  n10 = Constant2Vector() {g=0.2, r=0.2}
  n11 = LinearInterpolate(A=@n10, B=Multiply(A=@n10, B=ScalarParameter() {defaultValue=1, parameterName=OutlineSize}), Alpha=NamedRerouteUsage())
  MaterialFunctionCall(UVs=@n9.Skewed UVs, Texture Scale=@n11) {materialFunction=&ScaleUVsByCenter}
  n12 = Constant2Vector() {g=3, r=1}
  n13 = Multiply(A=TextureCoordinate(), B=@n12)
  n14 = ScalarParameter() {defaultValue=0.8, parameterName=Scale}
  n15 = Multiply(A=@n14, B=@n12.R)
  n16 = Multiply(A=@n12.G, B=@n14)
  NamedRerouteDeclaration(None=MaterialFunctionCall(UVs=Frac(None=@n13), Texture Scale=AppendVector(A=@n15, B=@n16)) {materialFunction=&ScaleUVsByCenter}.UVs) {name=UVMain, nodeColor=<rgba 0 0 0 1>}
  n17 = TextureCoordinate()
  n18 = AppendVector(
   A=Add(A=ComponentMask(None=@n17) {r=true}, B=ScalarParameter() {defaultValue=0.348, parameterName=PipX})
   B=Add(A=ComponentMask(None=@n17) {g=true}, B=ScalarParameter() {defaultValue=-0.344, parameterName=PipY})
  )
  n19 = Constant2Vector() {r=10}
  n20 = MaterialFunctionCall(
   UVs=MaterialFunctionCall(
    UVs=MaterialFunctionCall(UV=@n18, Skew Angle=@n19) {materialFunction=&MF_UI_Skew}.Skewed UVs
    "Rotation Angle (0-1)"=Constant() {r=0.125}
   ) {materialFunction=&CustomRotator}.Rotated Values
   Size=ScalarParameter() {defaultValue=0.12, parameterName=PipSize}
   OutlineThickness=Constant() {r=0.028}
  ) {materialFunction=&MF_UI_Box}
  NamedRerouteDeclaration(None=OneMinus(None=SmoothStep(Value=@n20.SDF) {constMax=0.001})) {name=Pip, nodeColor=<rgba 0 0 0 1>}
  n21 = NamedRerouteUsage()
  n22 = Add(A=ComponentMask(None=@n21) {r=true}, B=Constant() {r=1.19})
  n23 = ComponentMask(None=@n21) {g=true}
  n24 = SmoothStep(Value=MaterialFunctionCall(UVs=AppendVector(A=@n22, B=@n23), Size=Constant2Vector() {g=0.01, r=2}) {materialFunction=&MF_UI_Box}.SDF) {constMax=0.136}
  n25 = TextureCoordinate()
  n26 = ComponentMask(None=@n25) {r=true}
  n27 = Constant() {r=0.05}
  n28 = MaterialFunctionCall!(UV=AppendVector(A=Add(A=@n26, B=@n27), B=ComponentMask(None=@n25) {g=true}), Offset=Constant2Vector() {r=0.1}, Rotation, Scale, MirrorY) {materialFunction=/FortUILibrary/MaterialFunctions/Basics/MF_UI_Symmetry_2x1_DEPRECATED}
  NamedRerouteDeclaration(None=Multiply(
    A=OneMinus(None=@n24)
    B=MaterialFunctionCall(UVs=@n28.Result) {materialFunction=&MF_UI_LinearGradient}.Result
   ) {constA=1}) {name=Lines, nodeColor=<rgba 0 0 0 1>}
  n29 = NamedRerouteUsage()
  n30 = ComponentMask(None=@n29) {r=true}
  n31 = Constant() {r=1.15}
  n32 = ComponentMask(None=@n29) {g=true}
  n33 = TextureSample(UVs=Panner(Coordinate=NamedRerouteUsage()) {speedX=-8, speedY=1}) {texture=/ScubaBondOakGameplay/UI/Textures/T_UI_RTL_PerlinNoise}
  n34 = Add(A=AppendVector(A=Add(A=@n30, B=@n31), B=Subtract(A=@n32) {constB=0.01}), B=Multiply(A=AppendVector(A=@n33.R, B=@n33.R)) {constB=0.056})
  n35 = Constant2Vector() {g=0.02, r=2}
  n36 = OneMinus(None=SmoothStep(Value=MaterialFunctionCall(UVs=@n34, Size=@n35) {materialFunction=&MF_UI_Box}.SDF) {constMax=0.048})
  n37 = MaterialFunctionCall(GradientStart=Constant() {r=0.04}, GradientEnd=Constant() {r=0.328}) {materialFunction=&MF_UI_LinearGradient}
  NamedRerouteDeclaration(None=Multiply(A=Multiply(A=@n36, B=@n37.Result) {constA=1}, B=NamedRerouteUsage())) {name=Lines_progress, nodeColor=<rgba 0 0 0 1>}
  n38 = ScalarParameter() {group=Progress_Layout, parameterName=ProgressEdgeSharpness, sliderMax=1, sortPriority=2}
  n39 = MaterialFunctionCall(UVs=TextureCoordinate(), "Rotation Angle (0-1)"=Constant()) {materialFunction=&CustomRotator}
  n40 = ScalarParameter() {
   desc="When progress value changes, it defines where change starts.\r\n0 - 1"
   group=Progress
   parameterName=PipProgressStart
   sliderMax=1
   sortPriority=1
  }
  n41 = ScalarParameter() {
   defaultValue=0.76
   desc="When progress value changes, it defines where change ends.\r\n0 - 1"
   group=Progress
   parameterName=PipProgressEnd
   sliderMax=1
   sortPriority=2
  }
  n42 = ScalarParameter() {
   defaultValue=1
   desc="Property for animating progress value change (from ProgressStart to ProgressEnd)\r\n0 - 1"
   group=Progress
   parameterName=AnimatePipProgress
   sliderMax=1
   sortPriority=3
  }
  n43 = LinearInterpolate(A=@n40, B=@n41, Alpha=@n42)
  n44 = ComponentMask(None=Add(A=Multiply(A=@n39.Rotated Values) {constB=-1}, B=NamedRerouteDeclaration(None=@n43) {name=Progress_1, nodeColor=<rgba 0 0 0 1>}) {constB=0.5}) {r=true}
  NamedRerouteDeclaration(None=Saturate(None=SmoothStep(Max=@n38, Value=@n44) {constMax=0.01})) {name=ProgressGraident_Horizontal, nodeColor=<rgba 0 0 0 1>}
  n45 = Constant() {r=1}
  n46 = ScalarParameter() {parameterName=IsolatedUVIndex}
  NamedRerouteDeclaration(None=MaterialFunctionCall!(UV, Grid Size=Constant2Vector() {g=3, r=1}, Index=AppendVector(A=@n45, B=@n46)) {materialFunction=/FortUILibrary/MaterialFunctions/Basics/MF_UI_IsolateUV}.Result) {name=IsolatedUV, nodeColor=<rgba 0 0 0 1>}
  n47 = Constant2Vector() {g=3, r=1}
  n48 = Constant2Vector() {g=3, r=1}
  n49 = Add(
   A=Multiply(
    A=NamedRerouteUsage()
    B=MaterialFunctionCall(Grid Size=@n47, Index=@n48) {materialFunction=&MF_UI_IsolateUV}.Result
   )
   B=Constant2Vector() {r=-0.34}
  )
  n50 = Constant2Vector() {g=3, r=1}
  n51 = Constant2Vector() {g=2, r=1}
  n52 = Add(
   A=Multiply(
    A=NamedRerouteUsage()
    B=MaterialFunctionCall(Grid Size=@n50, Index=@n51) {materialFunction=&MF_UI_IsolateUV}.Result
   )
   B=Constant2Vector() {r=-0.4}
  )
  n53 = Multiply(
   A=NamedRerouteUsage()
   B=MaterialFunctionCall(Grid Size=Constant2Vector() {g=3, r=1}, Index=Constant2Vector() {g=1, r=1}) {materialFunction=&MF_UI_IsolateUV}.Result
  )
  n54 = Constant2Vector() {r=-0.45}
  NamedRerouteDeclaration(None=Max(A=Max(A=@n49, B=@n52), B=Add(A=@n53, B=@n54))) {name=UV, nodeColor=<rgba 0 0 0 1>}
  n55 = MaterialFunctionCall(
   UVs=MaterialFunctionCall(UVs=@n9.Skewed UVs, "Rotation Angle (0-1)"=Constant() {r=0.125}) {materialFunction=&CustomRotator}.Rotated Values
   Size=@n11
   OutlineThickness=Constant() {r=0.028}
  ) {materialFunction=&MF_UI_Box}
  NamedRerouteDeclaration(None=OneMinus(None=SmoothStep(Value=@n55.SDFOutline) {constMax=0.001})) {name=Outline, nodeColor=<rgba 0 0 0 1>}
  n56 = NamedRerouteUsage()
  n57 = Constant2Vector() {r=10}
  n58 = MaterialFunctionCall(
   UVs=MaterialFunctionCall(
    UVs=MaterialFunctionCall(UV=@n56, Skew Angle=@n57) {materialFunction=&MF_UI_Skew}.Skewed UVs
    "Rotation Angle (0-1)"=Constant() {r=0.125}
   ) {materialFunction=&CustomRotator}.Rotated Values
   Size=Constant2Vector() {g=0.16, r=0.16}
   OutlineThickness=Constant() {r=0.028}
  ) {materialFunction=&MF_UI_Box}
  NamedRerouteDeclaration(None=OneMinus(None=SmoothStep(Value=@n58.SDF) {constMax=0.001})) {name=Fill, nodeColor=<rgba 0 0 0 1>}
  n59 = ScalarParameter() {group=Progress_Layout, parameterName=ProgressEdgeSharpness, sliderMax=1, sortPriority=2}
  n60 = MaterialFunctionCall(UVs=TextureCoordinate(), "Rotation Angle (0-1)"=Constant() {r=0.5}) {materialFunction=&CustomRotator}
  n61 = ScalarParameter() {
   desc="When progress value changes, it defines where change starts.\r\n0 - 1"
   group=Progress
   parameterName=ProgressStart
   sliderMax=1
   sortPriority=1
  }
  n62 = ScalarParameter() {
   defaultValue=1
   desc="When progress value changes, it defines where change ends.\r\n0 - 1"
   group=Progress
   parameterName=ProgressEnd
   sliderMax=1
   sortPriority=2
  }
  n63 = ScalarParameter() {
   desc="Property for animating progress value change (from ProgressStart to ProgressEnd)\r\n0 - 1"
   group=Progress
   parameterName=AnimateProgress
   sliderMax=1
   sortPriority=3
  }
  n64 = LinearInterpolate(A=@n61, B=@n62, Alpha=@n63)
  n65 = ComponentMask(None=Add(A=Multiply(A=@n60.Rotated Values) {constB=-1}, B=NamedRerouteDeclaration(None=@n64) {name=Progress, nodeColor=<rgba 0 0 0 1>}) {constB=0.5}) {g=true}
  NamedRerouteDeclaration(None=Saturate(None=SmoothStep(Max=@n59, Value=@n65) {constMax=0.01})) {name=ProgressGraident, nodeColor=<rgba 0 0 0 1>}
  NamedRerouteUsage()
 ]
}
