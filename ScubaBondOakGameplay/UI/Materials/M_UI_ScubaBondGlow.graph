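# Translucent UI material that draws a cyan circular glow, named for the ScubaBond element of the game's UI.
# A Size-driven circle SDF from MF_UI_Circle is inverted and smoothstepped into a soft-edged mask.
# One product tints that mask cyan; another multiplies it by an inverted exponential radial gradient, fading the center.
# The Glow Intensity parameter sets the gradient's density.
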Material /ScubaBondOakGameplay/UI/Materials/M_UI_ScubaBondGlow
graph {
 properties {
  bAutomaticallySetUsageInEditor=false
  blendMode=BLEND_Translucent
  materialDomain=MD_UI
  pixelDepthOffsetMode=PDOM_Legacy
  usedShadingModels=Default Lit
 }
 nodes [
  n1 = ScalarParameter() {parameterName=Size}
  n2 = SmoothStep(Value=OneMinus(None=MaterialFunctionCall(Size=@n1) {materialFunction=&MF_UI_Circle}.SDF)) {constMax=0.2, constMin=0.1}
  n3 = ScalarParameter() {defaultValue=15, parameterName=Glow Intensity}
  Multiply(
   A=@n2
   B=OneMinus(None=MaterialFunctionCall(Density=@n3) {materialFunction=&RadialGradientExponential}.RadialGradientExponential)
  )
  Multiply(A=Constant3Vector() {constant=<rgba 0 1 1 0>}, B=@n2)
 ]
}
