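# Translucent UI material for the ScubaBondOak dash effect, drawing mirrored lightning with a tintable fill color and adjustable opacity.
# A lightning texture is sampled through deprecated 2x1 symmetry UVs with a horizontal offset parameter and center-based scaling.
# Its red channel feeds a distance-field outline (Glow) and a smoothstep mask (VFX).
# Rerouted masks drive lerps that blend the fill color and an opacity-weighted value.
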Material /ScubaBondOakGameplay/UI/Materials/M_UI_ScubaBondOak_DashVFX
graph {
 properties {
  bAutomaticallySetUsageInEditor=false
  blendMode=BLEND_Translucent
  materialDomain=MD_UI
  pixelDepthOffsetMode=PDOM_Legacy
  usedShadingModels=Default Lit
 }
 nodes [
  n1 = NamedRerouteUsage()
  LinearInterpolate(A=NamedRerouteUsage(), B=Multiply(A=@n1, B=ScalarParameter() {defaultValue=0.8, parameterName=Opacity}), Alpha=@n1)
  LinearInterpolate(A=VectorParameter() {parameterName=FillColor, defaultValue=<rgba 0 0.5239 1 1>}.RGB, Alpha=NamedRerouteUsage())
  n2 = AppendVector(A=ScalarParameter() {parameterName=OffsetX}, B=Constant())
  n3 = Constant2Vector() {g=0.5, r=1}
  n4 = TextureSample(UVs=MaterialFunctionCall(
    UVs=MaterialFunctionCall!(UV, Offset=@n2, Rotation, Scale=@n3, MirrorY) {materialFunction=/FortUILibrary/MaterialFunctions/Basics/MF_UI_Symmetry_2x1_DEPRECATED}.Result
    Texture Scale=ScalarParameter() {defaultValue=1, parameterName=Scale}
   ) {materialFunction=&ScaleUVsByCenter}.UVs) {samplerSource=SSM_Clamp_WorldGroupSettings, texture=/ScubaBondOakGameplay/UI/Textures/T_UI_Lightning_fx__1_}
  NamedRerouteDeclaration(None=MaterialFunctionCall!(RGBA Channels=@n4.R, Glow Min, Glow Max, Outline Thickness=Constant() {r=0.9959}) {materialFunction=/FortUILibrary/MaterialFunctions/Basics/MF_UI_ChannelPackedDistanceField}.DF Outline) {name=Glow, nodeColor=<rgba 0 0 0 1>}
  NamedRerouteDeclaration(None=SmoothStep(Value=@n4.R) {constMax=0.628, constMin=0.424}) {name=VFX, nodeColor=<rgba 0 0 0 1>}
 ]
}
