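# Translucent UI material for the ScubaBond reticle, built on a lightning-shaped signed distance field texture.
# A SmoothStep remaps the texture's red channel between fixed thresholds to form a soft-edged shape mask.
# An Opacity scalar parameter multiplies the mask so the reticle can be faded.
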Material /ScubaBondOakGameplay/UI/Materials/M_UI_ScubaBondReticle
graph {
 properties {
  bAutomaticallySetUsageInEditor=false
  blendMode=BLEND_Translucent
  materialDomain=MD_UI
  pixelDepthOffsetMode=PDOM_Legacy
  usedShadingModels=Default Lit
 }
 nodes [Multiply(
   A=SmoothStep(Value=TextureSample() {texture=/ScubaBondOakGameplay/UI/Textures/T_UI_Lightning_SDF}.R) {constMax=0.656, constMin=0.28}
   B=ScalarParameter() {defaultValue=1, parameterName=Opacity}
  )]
}
