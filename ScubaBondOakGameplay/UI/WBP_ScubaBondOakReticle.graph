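# Reticle extension widget for the ScubaBond Oak weapon that animates its jump cooldown with dash pips, glow and lightning.
# On construct it watches the view target's Active tag addition and Cooldown tag removal, or binds the weapon's cooldown delegate if bBindToGameplayTagAddition is off.
# SetCoolDownEvent plays the jump-start animation, or stops the pip animation and plays the finish animation.
# Sequencer events drive the dash reveal and seven-second pip sequence.
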WidgetBlueprint /ScubaBondOakGameplay/UI/WBP_ScubaBondOakReticle
animations [
 Anim_JumpStarted {
  binding_count=1
  length_seconds=3.0008
  object=&self:Anim_JumpStarted
  sequence {
   bindings [Image_ReticleVisual {
     id=#1
     possessed_class=Image
     tracks [
      {
       type=MovieSceneWidgetMaterialTrack
       sections [{type=MovieSceneParameterSection, end_frame=60, start_frame=0, channels [Opacity {type=MovieSceneScriptingFloatChannel, keys [[0,1],[5,0]]}]}]
      }
      {
       type=MovieScene2DTransformTrack
       sections [{
         type=MovieScene2DTransformSection
         end_frame=60
         start_frame=0
         channels [
          Scale.X {type=MovieSceneScriptingFloatChannel, keys [[0,1],[4.996,4]]}
          Scale.Y {type=MovieSceneScriptingFloatChannel, keys [[0,1],[4.996,4]]}
         ]
        }]
      }
     ]
    }]
   tracks [{
     type=MovieSceneEventTrack
     sections [{
       type=MovieSceneEventTriggerSection
       channels [None {
         type=MovieSceneScriptingEventChannel
         keys [[5, "<Struct 'MovieSceneEvent' (0x46911f7ece0) {}>"], [6, "<Struct 'MovieSceneEvent' (0x46911f7ee20) {}>"]]
        }]
      }]
    }]
  }
 }
 Anim_AnimatePips {
  binding_count=5
  length_seconds=7.05
  object=&self:Anim_AnimatePips
  sequence {bindings [
    Image_ReticleVisual {
     id=#2
     possessed_class=Image
     tracks [
      {
       type=MovieSceneWidgetMaterialTrack
       sections [{
         type=MovieSceneParameterSection
         end_frame=140
         start_frame=-1
         channels [Opacity {type=MovieSceneScriptingFloatChannel, keys [[-1,0],[135,0],[140,0.5]]}]
        }]
      }
      {
       type=MovieScene2DTransformTrack
       sections [{
         type=MovieScene2DTransformSection
         end_frame=140
         start_frame=110
         channels [
          Scale.X {type=MovieSceneScriptingFloatChannel, keys [[135,1.1],[140,1]]}
          Scale.Y {type=MovieSceneScriptingFloatChannel, keys [[135,1.1],[140,1]]}
         ]
        }]
      }
     ]
    }
    Image_LigtningFx {
     id=#3
     possessed_class=Image
     tracks [
      {
       type=MovieSceneFloatTrack
       display_name=Render Opacity
       sections [{
         type=MovieSceneFloatSection
         end_frame=117
         start_frame=-1
         channels [None {type=MovieSceneScriptingFloatChannel, keys [[-1,0],[109,0],[113,1],[116,0]]}]
        }]
      }
      {
       type=MovieSceneWidgetMaterialTrack
       sections [{
         type=MovieSceneParameterSection
         end_frame=117
         start_frame=-1
         channels [
          Opacity {type=MovieSceneScriptingFloatChannel, const [0.8,2]}
          Scale {type=MovieSceneScriptingFloatChannel, keys [[-1,0.2],[109,0.2],[113,1],[116,0.2]]}
          OffsetX {type=MovieSceneScriptingFloatChannel, keys [[109,-0.2],[113,0],[116,-0.2]]}
         ]
        }]
      }
     ]
    }
    Image_GlowFX {
     id=#4
     possessed_class=Image
     tracks [
      {
       type=MovieScene2DTransformTrack
       sections [
        {
         type=MovieScene2DTransformSection
         end_frame=121
         start_frame=-1
         channels [
          Translation.X {type=MovieSceneScriptingFloatChannel, keys [[-1,80.4139],[35,93.698],[64,105.6528],[94,122.7592]]}
          Translation.Y {type=MovieSceneScriptingFloatChannel, keys [[-1,96.3606],[35,47.8914],[64,0.6478],[94,-60.5427]]}
         ]
        }
        {
         type=MovieScene2DTransformSection
         end_frame=121
         start_frame=-1
         channels [
          Scale.X {type=MovieSceneScriptingFloatChannel, keys [[-1,0],[27,0.3],[35,1.1],[56,0.3],[64,1.1],[85,0.3],[93,1.1],[109,0.3],[121,1.5]]}
          Scale.Y {type=MovieSceneScriptingFloatChannel, keys [[-1,0],[27,0.3],[35,1.1],[56,0.3],[64,1.1],[85,0.3],[93,1.1],[109,0.3],[121,1.5]]}
         ]
        }
       ]
      }
      {
       type=MovieSceneWidgetMaterialTrack
       sections [{
         type=MovieSceneParameterSection
         end_frame=121
         start_frame=-1
         channels [
          Glow Intensity {
           type=MovieSceneScriptingFloatChannel
           keys [[-1,15],[27,15],[28,8],[31,4],[35,15],[56,15],[57,8],[60,4],[64,15],[85,15],[86,8],[89,4],[93,15],[109,15],[115,4],[121,15]]
          }
          Size {
           type=MovieSceneScriptingFloatChannel
           keys [[-1,0],[27,0],[33,0.16],[35,0],[56,0],[62,0.16],[64,0],[85,0],[91,0.16],[93,0],[109,0],[115,0.16],[121,0]]
          }
         ]
        }]
      }
     ]
    }
    Image_ReticleVisual_Pips {
     id=#5
     possessed_class=Image
     tracks [
      {
       type=MovieSceneWidgetMaterialTrack
       sections [{
         type=MovieSceneParameterSection
         channels [
          AnimateProgress {type=MovieSceneScriptingFloatChannel, keys [[-1,0],[29,0.3],[57,0.6],[84,1]]}
          PipOpacity {type=MovieSceneScriptingFloatChannel, keys [[-1,0],[1,1],[29,0.8],[31,0],[32,1],[33,1],[58,0.8],[60,0],[61,1],[62,1],[84,0.8],[86,0]]}
          PipSize {
           type=MovieSceneScriptingFloatChannel
           keys [[25,0.12],[29,0.25],[31,0],[32,0.12],[54,0.12],[58,0.25],[60,0],[61,0.12],[80,0.12],[84,0.25],[86,0],[87,0.12]]
          }
          PipX {type=MovieSceneScriptingFloatChannel, keys [[-1,0.348],[25,-0.282],[32,0.348],[54,-0.322],[61,0.348],[80,-0.362]]}
          PipY {type=MovieSceneScriptingFloatChannel, keys [[-1,-0.334],[25,-0.334],[32,-0.004],[61,0.326]]}
          IsolatedUVIndex {type=MovieSceneScriptingFloatChannel, keys [[-1,0],[0,3],[32,2],[62,1],[88,0]]}
          AnimatePipProgress {type=MovieSceneScriptingFloatChannel, keys [[-1,0],[21,1],[31,0],[32,0],[53,1],[61,0],[62,0],[78,1],[88,0]]}
          OutlineSize {type=MovieSceneScriptingFloatChannel, keys [[-1,1],[28,1],[30,1.3],[32,1],[58,1],[60,1.3],[62,1],[84,1],[86,1.3],[88,1]]}
         ]
        }]
      }
      {
       type=MovieSceneFloatTrack
       display_name=Render Opacity
       sections [{
         type=MovieSceneFloatSection
         end_frame=123
         start_frame=-1
         channels [None {type=MovieSceneScriptingFloatChannel, keys [[-1,1],[117,1],[122,0]]}]
        }]
      }
     ]
    }
    Image_ReticleVisual_TopPip {
     id=#6
     possessed_class=Image
     tracks [
      {
       type=MovieSceneWidgetMaterialTrack
       sections [{
         type=MovieSceneParameterSection
         end_frame=118
         start_frame=-1
         channels [
          LightningX {type=MovieSceneScriptingFloatChannel, keys [[81,-0.104],[86,-0.024],[89,-0.024]]}
          LightningY {type=MovieSceneScriptingFloatChannel, keys [[81,0.7],[86,-0.05],[89,0.008]]}
          LightningOpacity {type=MovieSceneScriptingFloatChannel, keys [[81,0],[86,1]]}
          ScaleTexture {type=MovieSceneScriptingFloatChannel, keys [[81,0],[86,1],[109,1],[112,0.64],[115,1]]}
          AnimateProgress {type=MovieSceneScriptingFloatChannel, keys [[89,0],[109,1]]}
          PipScale {type=MovieSceneScriptingFloatChannel, keys [[-1,0.9],[109,0.9],[112,0.6],[115,0.9]]}
         ]
        }]
      }
      {
       type=MovieSceneFloatTrack
       display_name=Render Opacity
       sections [{
         type=MovieSceneFloatSection
         end_frame=124
         start_frame=-1
         channels [None {type=MovieSceneScriptingFloatChannel, keys [[-1,1],[117,1],[122,0]]}]
        }]
      }
     ]
    }
   ]}
 }
 Anim_DashesReveal {
  binding_count=2
  length_seconds=0.15
  object=&self:Anim_DashesReveal
  sequence {bindings [
    Image_ReticleVisual_TopPip {
     id=#7
     possessed_class=Image
     tracks [
      {
       type=MovieSceneFloatTrack
       display_name=Render Opacity
       sections [{type=MovieSceneFloatSection, end_frame=3, start_frame=0, channels [None {type=MovieSceneScriptingFloatChannel, keys [[0,0],[3,1]]}]}]
      }
      {
       type=MovieScene2DTransformTrack
       sections [{
         type=MovieScene2DTransformSection
         end_frame=3
         start_frame=0
         channels [
          Translation.X {type=MovieSceneScriptingFloatChannel, keys [[0,156],[3,80]]}
          Translation.Y {type=MovieSceneScriptingFloatChannel, keys [[0,-142],[3,-112]]}
         ]
        }]
      }
     ]
    }
    Image_ReticleVisual_Pips {
     id=#8
     possessed_class=Image
     tracks [
      {
       type=MovieSceneFloatTrack
       display_name=Render Opacity
       sections [{type=MovieSceneFloatSection, end_frame=3, start_frame=0, channels [None {type=MovieSceneScriptingFloatChannel, keys [[0,0],[3,1]]}]}]
      }
      {
       type=MovieScene2DTransformTrack
       sections [{
         type=MovieScene2DTransformSection
         end_frame=3
         start_frame=0
         channels [
          Translation.X {type=MovieSceneScriptingFloatChannel, keys [[0,74],[3,0]]}
          Translation.Y {type=MovieSceneScriptingFloatChannel, const [0,1]}
         ]
        }]
      }
     ]
    }
   ]}
 }
 Anim_CooldownFinished {
  binding_count=4
  length_seconds=0.15
  object=&self:Anim_CooldownFinished
  sequence {bindings [
    Image_GlowFX {
     id=#9
     possessed_class=Image
     tracks [{
       type=MovieSceneWidgetMaterialTrack
       sections [{
         type=MovieSceneParameterSection
         end_frame=3
         start_frame=0
         channels [
          Glow Intensity {type=MovieSceneScriptingFloatChannel, const [15,1]}
          Size {type=MovieSceneScriptingFloatChannel, const [0,1]}
         ]
        }]
      }]
    }
    Image_ReticleVisual {
     id=#10
     possessed_class=Image
     tracks [{
       type=MovieSceneWidgetMaterialTrack
       sections [{type=MovieSceneParameterSection, end_frame=3, start_frame=0, channels [Opacity {type=MovieSceneScriptingFloatChannel, keys [[0,0.5],[3,1]]}]}]
      }]
    }
    Image_ReticleVisual_TopPip {
     id=#11
     possessed_class=Image
     tracks [{
       type=MovieSceneWidgetMaterialTrack
       sections [{
         type=MovieSceneParameterSection
         end_frame=3
         start_frame=0
         channels [
          LightningX {type=MovieSceneScriptingFloatChannel, const [-0.104,2]}
          LightningY {type=MovieSceneScriptingFloatChannel, const [0.7,2]}
          LightningOpacity {type=MovieSceneScriptingFloatChannel, const [0,1]}
          ScaleTexture {type=MovieSceneScriptingFloatChannel, const [0,1]}
          AnimateProgress {type=MovieSceneScriptingFloatChannel, const [0,1]}
         ]
        }]
      }]
    }
    Image_ReticleVisual_Pips {
     id=#12
     possessed_class=Image
     tracks [{
       type=MovieSceneWidgetMaterialTrack
       sections [{
         type=MovieSceneParameterSection
         end_frame=3
         start_frame=0
         channels [
          AnimateProgress {type=MovieSceneScriptingFloatChannel, const [0,1]}
          PipX {type=MovieSceneScriptingFloatChannel, const [0.348,1]}
          PipY {type=MovieSceneScriptingFloatChannel, const [-0.344,1]}
          PipSize {type=MovieSceneScriptingFloatChannel, const [0.12,1]}
          PipOpacity {type=MovieSceneScriptingFloatChannel, const [0,1]}
          IsolatedUVIndex {type=MovieSceneScriptingFloatChannel, const [0,1]}
          AnimatePipProgress {type=MovieSceneScriptingFloatChannel, const [0,1]}
         ]
        }]
      }]
    }
   ]}
 }
]
graphs [
 EventGraph {nodes [
   Event Construct()
   > n1 = Cast To B_Weap_ScubaBondOak!(
    execute
    Object=GetEquippedWeapon!(self:Fort Weapon Reticle Extension Widget Base Object Reference) -> (ReturnValue:Fort Weapon Object Reference)
   ) -> (then, CastFailed, AsB Weap Scuba Bond Oak:B Weap Scuba Bond Oak Object Reference) {is=title K2Node_DynamicCast}
   > n2 = Branch(Condition=Get bBindToGameplayTagAddition() -> (bBindToGameplayTagAddition:bool))
   > n3 = Cast To FortPlayerController(Object=GetOwningPlayer())
   n4 = GetViewTarget!(self:Controller Object Reference=@n3) -> (ReturnValue:Actor Object Reference)
   > n5 = WaitGameplayTagAddToActor!(
    execute:Exec
    TargetActor:Actor Object Reference=@n4
    Tag:Gameplay Tag Structure=(TagName="Granted.ScubaBond.Oak.Active")
    OnlyTriggerOnce:bool=false
   ) -> (then:Exec, AsyncTaskProxy:Ability Async Wait Gameplay Tag Added Object Reference, Added:Exec) {is=title K2Node_AsyncAction}
   SetCoolDownEvent!(execute=@n5.Added, self, IsCooldown?:bool=true) -> (then)
   > n6 = WaitGameplayTagRemoveFromActor!(
    execute:Exec
    TargetActor:Actor Object Reference=@n4
    Tag:Gameplay Tag Structure=(TagName="Ability.ScubaBond.Oak.Cooldown")
    OnlyTriggerOnce:bool=false
   ) -> (then:Exec, AsyncTaskProxy:Ability Async Wait Gameplay Tag Removed Object Reference, Removed:Exec) {is=title K2Node_AsyncAction}
   SetCoolDownEvent(execute=@n6.Removed, IsCooldown?=false)
   n7 = Set WeaponRef(execute=@n2.else, WeaponRef:B Weap Scuba Bond Oak Object Reference=@n1) -> (Output_Get:B Weap Scuba Bond Oak Object Reference)
   > Bind Event to On Cooldown Changed(self:B Weap Scuba Bond Oak Object Reference=@n7, Delegate=Create Event()) {is=title K2Node_AddDelegate}
   Event Destruct()
   > n8 = Branch(Condition=Get bBindToGameplayTagAddition())
   Unbind all Events from On Cooldown Changed!(execute:Exec=@n8.else, self:B Weap Scuba Bond Oak Object Reference=Get WeaponRef() -> (WeaponRef:B Weap Scuba Bond Oak Object Reference)) -> (then:Exec) {is=title K2Node_ClearDelegate}
  ]}
 Sequencer Events {nodes [
   RevealDashes() {is=title K2Node_CustomEvent}
   > PlayAnimationForward(InAnimation=Get Anim_DashesReveal() -> (Anim_DashesReveal:Widget Animation Object Reference))
   StartJumpSequence() {is=title K2Node_CustomEvent}
   > PlayAnimation(InAnimation=Get Anim_AnimatePips() -> (Anim_AnimatePips:Widget Animation Object Reference))
  ]}
 SetCoolDownEvent {nodes [
   n1 = SetCoolDownEvent!() -> (then, IsCooldown?:bool=~) {is=title K2Node_FunctionEntry}
   > n2 = Branch(Condition=@n1)
   > PlayAnimationForward(InAnimation=Get Anim_JumpStarted() -> (Anim_JumpStarted:Widget Animation Object Reference))
   StopAnimation(execute=@n2.else, InAnimation=Get Anim_AnimatePips())
   > PlayAnimationForward(InAnimation=Get Anim_CooldownFinished() -> (Anim_CooldownFinished:Widget Animation Object Reference))
  ]}
]
variables [
 bBindToGameplayTagAddition {type=boolean, default=true}
 WeaponRef {type=B_Weap_ScubaBondOak_C, default=null}
]
widgets {
 info {
  inheritedWidgetCount=1
  parentClass=/Script/FortniteUI.FortWeaponReticleExtensionWidgetBase
  rootWidgetClass=&Overlay
  widgetCount=7
 }
 widgets [
  Reticle {widgetClassPath=&Overlay, widget=Reticle}
  Image_LigtningFx {
   widgetClassPath=&Image
   bIsVariable=true
   parent=Reticle
   widget=Image_LigtningFx
   properties {
    renderOpacity=0
    brush {
     bIsDynamicallyLoaded=false
     drawAs=Image
     imageType=FullColor
     mirroring=NoMirror
     resourceName=None
     resourceObject=/ScubaBondOakGameplay/UI/Materials/M_UI_ScubaBondOak_DashVFX
     tiling=NoTile
     imageSize=<100 100>
     margin=<ltrb 0 0 0 0>
     outlineSettings {
      bUseBrushTransparency=false
      roundingType=HalfHeightRadius
      width=0
      color {colorUseRule=UseColor_Specified, specifiedColor=<rgba 0 0 0 0>}
      cornerRadii=<0 0 0 0>
     }
     tintColor {colorUseRule=UseColor_Specified, specifiedColor=<rgba 1 1 1 1>}
     uVRegion {bIsValid=false, max=<0 0>, min=<0 0>}
    }
    renderTransform {angle=0, scale=<0.9 0.9>, shear=<0 0>, translation=<120 -76>}
   }
   slot {class=OverlaySlot, properties {padding=<ltrb 135 20 0 0>}}
  }
  Image_ReticleVisual {
   widgetClassPath=&Image
   bIsVariable=true
   parent=Reticle
   widget=Image_ReticleVisual
   properties {brush {
     bIsDynamicallyLoaded=false
     drawAs=Image
     imageType=FullColor
     mirroring=NoMirror
     resourceName=None
     resourceObject=/ScubaBondOakGameplay/UI/Materials/M_UI_ScubaBondReticle
     tiling=NoTile
     imageSize=<40 40>
     margin=<ltrb 0 0 0 0>
     outlineSettings {
      bUseBrushTransparency=false
      roundingType=HalfHeightRadius
      width=0
      color {colorUseRule=UseColor_Specified, specifiedColor=<rgba 0 0 0 0>}
      cornerRadii=<0 0 0 0>
     }
     tintColor {colorUseRule=UseColor_Specified, specifiedColor=<rgba 1 1 1 1>}
     uVRegion {bIsValid=false, max=<0 0>, min=<0 0>}
    }}
   slot {class=OverlaySlot, properties {horizontalAlignment=HAlign_Center, verticalAlignment=VAlign_Center, padding=<ltrb 0 0 50 10>}}
  }
  Image_ReticleVisual_Pips {
   widgetClassPath=&Image
   bIsVariable=true
   parent=Reticle
   widget=Image_ReticleVisual_Pips
   properties {
    renderOpacity=0
    brush {
     bIsDynamicallyLoaded=false
     drawAs=Image
     imageType=FullColor
     mirroring=NoMirror
     resourceName=None
     resourceObject=/ScubaBondOakGameplay/UI/Materials/M_UI_ScubaBondDashMaterial
     tiling=NoTile
     imageSize=<160 160>
     margin=<ltrb 0 0 0 0>
     outlineSettings {
      bUseBrushTransparency=false
      roundingType=HalfHeightRadius
      width=0
      color {colorUseRule=UseColor_Specified, specifiedColor=<rgba 0 0 0 0>}
      cornerRadii=<0 0 0 0>
     }
     tintColor {colorUseRule=UseColor_Specified, specifiedColor=<rgba 1 1 1 1>}
     uVRegion {bIsValid=false, max=<0 0>, min=<0 0>}
    }
    renderTransform {angle=0, scale=<0.9 0.9>, shear=<-7 0>, translation=<0 0>}
   }
   slot {class=OverlaySlot, properties {horizontalAlignment=HAlign_Center, verticalAlignment=VAlign_Center, padding=<ltrb 100 20 0 0>}}
  }
  Image_ReticleVisual_TopPip {
   widgetClassPath=&Image
   bIsVariable=true
   parent=Reticle
   widget=Image_ReticleVisual_TopPip
   properties {
    renderOpacity=0
    brush {
     bIsDynamicallyLoaded=false
     drawAs=Image
     imageType=FullColor
     mirroring=NoMirror
     resourceName=None
     resourceObject=/ScubaBondOakGameplay/UI/Materials/M_UI_LastDash
     tiling=NoTile
     imageSize=<80 80>
     margin=<ltrb 0 0 0 0>
     outlineSettings {
      bUseBrushTransparency=false
      roundingType=HalfHeightRadius
      width=0
      color {colorUseRule=UseColor_Specified, specifiedColor=<rgba 0 0 0 0>}
      cornerRadii=<0 0 0 0>
     }
     tintColor {colorUseRule=UseColor_Specified, specifiedColor=<rgba 1 1 1 1>}
     uVRegion {bIsValid=false, max=<0 0>, min=<0 0>}
    }
    renderTransform {angle=0, scale=<0.9 0.9>, shear=<-7 0>, translation=<80 -112>}
   }
   slot {class=OverlaySlot, properties {horizontalAlignment=HAlign_Center, verticalAlignment=VAlign_Center, padding=<ltrb 95 20 0 0>}}
  }
  Image_GlowFX {
   widgetClassPath=&Image
   bIsVariable=true
   parent=Reticle
   widget=Image_GlowFX
   properties {
    brush {
     bIsDynamicallyLoaded=false
     drawAs=Image
     imageType=FullColor
     mirroring=NoMirror
     resourceName=None
     resourceObject=/ScubaBondOakGameplay/UI/Materials/M_UI_ScubaBondGlow
     tiling=NoTile
     imageSize=<64 64>
     margin=<ltrb 0 0 0 0>
     outlineSettings {
      bUseBrushTransparency=false
      roundingType=HalfHeightRadius
      width=0
      color {colorUseRule=UseColor_Specified, specifiedColor=<rgba 0 0 0 0>}
      cornerRadii=<0 0 0 0>
     }
     tintColor {colorUseRule=UseColor_Specified, specifiedColor=<rgba 1 1 1 1>}
     uVRegion {bIsValid=false, max=<0 0>, min=<0 0>}
    }
    renderTransform {angle=0, scale=<0.8 0.8>, shear=<0 0>, translation=<64 1>}
   }
   slot {class=OverlaySlot, properties {padding=<ltrb 150 30 0 0>}}
  }
  TutorialHighlight {widgetClassPath=&TutorialHighlight}
 ]
}
